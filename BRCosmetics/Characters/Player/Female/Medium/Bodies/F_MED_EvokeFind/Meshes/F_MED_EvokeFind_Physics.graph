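# Physics asset for the F_MED_EvokeFind female medium body cosmetic, covering a dangling tassel bone chain and the body around it.
# Capsules on the pelvis, three spine bones and both thighs form a coarse body shape, likely as collision for the tassel.
# A master tassel bone hangs from spine_02 with free rotation, followed by four chained capsule links whose constraints give narrower twist than swing limits.
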PhysicsAsset /BRCosmetics/Characters/Player/Female/Medium/Bodies/F_MED_EvokeFind/Meshes/F_MED_EvokeFind_Physics
bodies {
 dyn_tassle_01 [{shapeType=Capsule, length=2.1932, radius=0.922, shapeName=dyn_tassle_01_capsule, center=<2.1119 0 0>, rotation=<rot 90 0 0>}]
 dyn_tassle_02 [{shapeType=Capsule, length=3.8494, radius=1.7616, shapeName=dyn_tassle_02_capsule, center=<2.1119 0 0>, rotation=<rot 90 0 0>}]
 dyn_tassle_03 [{shapeType=Capsule, length=4.6755, radius=2.0306, shapeName=dyn_tassle_03_capsule, center=<2.1119 0 0>, rotation=<rot 90 0 0>}]
 dyn_tassle_04 [{shapeType=Capsule, length=2.6276, radius=2.492, shapeName=dyn_tassle_03_capsule, center=<1.3147 0 0>, rotation=<rot 90 0 0>}]
 dyn_tassle_mstr [{shapeType=Capsule, length=0.505, radius=0.505, shapeName=dyn_tassle_mstr_capsule, rotation=<rot 0 -90 0>}]
 pelvis [{shapeType=Capsule, length=6.6454, radius=10.3082, center=<0 -2.9132 0>}]
 spine_01 [{shapeType=Capsule, length=7.1412, radius=8.6333, shapeName=spine_01_capsule, center=<0 -1.6641 0>}]
 spine_02 [{shapeType=Capsule, length=6.3877, radius=6.017}]
 spine_03 [{shapeType=Capsule, length=3.6123, radius=6.4677, shapeName=spine_03_capsule, center=<-2.286 1.5842 0>}]
 thigh_l [{shapeType=Capsule, length=25.5856, radius=8.4144, center=<-10 -0.25 1.5>, rotation=<rot -75 0 0>}]
 thigh_r [{shapeType=Capsule, length=24.8182, radius=9.9498, center=<11.0061 3.0095 -2.0892>, rotation=<rot 80 179.9999 180>}]
}
constraints [
 {bone1Name=spine_02, bone2Name=spine_01, swing1LimitDegrees=45, swing2LimitDegrees=45, twistLimitDegrees=45}
 {bone1Name=thigh_l, bone2Name=pelvis, swing1LimitDegrees=45, swing2LimitDegrees=45, twistLimitDegrees=45}
 {bone1Name=thigh_r, bone2Name=pelvis, swing1LimitDegrees=45, swing2LimitDegrees=45, twistLimitDegrees=45}
 {
  bone1Name=spine_01
  bone2Name=pelvis
  swing1LimitDegrees=45
  swing1Motion=Free
  swing2LimitDegrees=45
  swing2Motion=Free
  twistLimitDegrees=45
  twistMotion=Free
 }
 {
  bone1Name=spine_03
  bone2Name=spine_02
  swing1LimitDegrees=45
  swing1Motion=Free
  swing2LimitDegrees=45
  swing2Motion=Free
  twistLimitDegrees=45
  twistMotion=Free
 }
 {
  bone1Name=dyn_tassle_mstr
  bone2Name=spine_02
  swing1LimitDegrees=45
  swing1Motion=Free
  swing2LimitDegrees=45
  swing2Motion=Free
  twistLimitDegrees=45
  twistMotion=Free
 }
 {bone1Name=dyn_tassle_01, bone2Name=dyn_tassle_mstr, swing1LimitDegrees=35, swing2LimitDegrees=35, twistLimitDegrees=10}
 {bone1Name=dyn_tassle_02, bone2Name=dyn_tassle_01, swing1LimitDegrees=35, swing2LimitDegrees=35, twistLimitDegrees=10}
 {bone1Name=dyn_tassle_03, bone2Name=dyn_tassle_02, swing1LimitDegrees=35, swing2LimitDegrees=35, twistLimitDegrees=10}
 {bone1Name=dyn_tassle_04, bone2Name=dyn_tassle_03, swing1LimitDegrees=35, swing2LimitDegrees=35, twistLimitDegrees=10}
]
