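# Physics asset for the face accessory mesh of the EvokeFind medium female player character, covering a left and right pair of dynamic bun bones.
# Each side has a larger offset capsule on the bun bone and a tiny capsule on its master bone.
# One constraint per side links bun to master, with matching swing and twist limit angles.
# The right bun's capsule keeps the left bun's shape name.
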PhysicsAsset /BRCosmetics/Characters/Player/Female/Medium/Bodies/F_MED_EvokeFind/Meshes/Parts/F_MED_EvokeFind_FaceAcc_Physics
bodies {
 dyn_bun_l [{
   shapeType=Capsule
   length=6.0454
   radius=3.0675
   shapeName=dyn_bun_l_capsule
   center=<0.1436 0.7221 -2.3835>
   rotation=<rot 0 -115.9135 -87.5257>
  }]
 dyn_bun_l_mstr [{shapeType=Capsule, length=0.505, radius=0.505, shapeName=dyn_bun_l_mstr_capsule, rotation=<rot 0 -90 0>}]
 dyn_bun_r [{shapeType=Capsule, length=6.0454, radius=3.0675, shapeName=dyn_bun_l_capsule, center=<-0.5063 0.419 -2.3459>, rotation=<rot 0 115 87>}]
 dyn_bun_r_mstr [{shapeType=Capsule, length=0.505, radius=0.505, shapeName=dyn_bun_r_mstr_capsule, rotation=<rot 0 -90 0>}]
}
constraints [
 {bone1Name=dyn_bun_r, bone2Name=dyn_bun_r_mstr, swing1LimitDegrees=30, swing2LimitDegrees=20, twistLimitDegrees=10}
 {bone1Name=dyn_bun_l, bone2Name=dyn_bun_l_mstr, swing1LimitDegrees=30, swing2LimitDegrees=20, twistLimitDegrees=10}
]
